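# Sound cue for the one-handed female Space Wanderer pickaxe hitting an enemy.
# A random node picks among six equally weighted impact wave variations.
# The cue plays them at raised pitch, routed through the Fort impacts sound class with shared pickaxe impact attenuation and a cosmetic pickaxe property sheet.
# It is primed on load.
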SoundCue /BRCosmetics/Sounds/PickAxes/SpaceWandererFemale/PickaxeImpactEnemy_SpaceWandererFemale1H_Cue
root=Random_0
nodes [
 {id=Random_0, children [WavePlayer_0,WavePlayer_1,WavePlayer_2,WavePlayer_3,WavePlayer_4,WavePlayer_5], properties {weights [1,1,1,1,1,1]}}
 {id=WavePlayer_0, properties {soundWaveAssetPtr=&/Space_Wanderer_Female_Impact_01}}
 {id=WavePlayer_1, properties {soundWaveAssetPtr=&/Space_Wanderer_Female_Impact_02}}
 {id=WavePlayer_2, properties {soundWaveAssetPtr=&/Space_Wanderer_Female_Impact_03}}
 {id=WavePlayer_3, properties {soundWaveAssetPtr=&/Space_Wanderer_Female_Impact_04}}
 {id=WavePlayer_4, properties {soundWaveAssetPtr=&/Space_Wanderer_Female_Impact_05}}
 {id=WavePlayer_5, properties {soundWaveAssetPtr=&/Space_Wanderer_Female_Impact_06}}
]
properties {
 attenuationSettings=/Game/Sounds/Attenuation/Fort_Pickaxe_Impact_Att
 audioPropertiesSheet=/BRCosmetics/Sounds/Settings/PropertySheets/Pickaxe/APS_Cosmetic_Pickaxe_Impact
 bPrimeOnLoad=true
 maxDistance=4300
 pitchMultiplier=1.5
 soundClassObject=&Fort_Impacts
}
